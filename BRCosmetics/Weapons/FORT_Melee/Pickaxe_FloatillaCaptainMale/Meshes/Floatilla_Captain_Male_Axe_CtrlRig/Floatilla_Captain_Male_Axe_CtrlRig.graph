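# Control Rig for the Floatilla Captain Male pickaxe that procedurally rotates its dyn_baseball bone.
# It reads the X Euler angle of dyn_baseball_MSTR_01 relative to b_AR_Root and remaps it, clamped, to a wider Y-axis angle.
# An accumulated quaternion lerp smooths that rotation before it is applied additively in local space.
# Rope and teeth bones are left untouched, and a Get Bone Transform node is unconnected.
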ControlRigBlueprint /BRCosmetics/Weapons/FORT_Melee/Pickaxe_FloatillaCaptainMale/Meshes/Floatilla_Captain_Male_Axe_CtrlRig/Floatilla_Captain_Male_Axe_CtrlRig
graphs [
 RigVMFunctionLibrary {links [], nodes []}
 RigVMModel {nodes [
   RigUnit_BeginExecution = &RigUnit_BeginExecution()
   n1 = &RigVMFunction_MathFloatRemap(
    SourceMaximum=85
    SourceMinimum=-18
    TargetMaximum=400
    TargetMinimum=0
    bClamp=true
    Value=&RigVMFunction_MathQuaternionToEuler(RotationOrder=ZYX, Value=&RigUnit_GetRelativeBoneTransform(Bone=dyn_baseball_MSTR_01, Space=b_AR_Root).Transform.Rotation).Result.X
   )
   /Script/ControlRig.RigUnit_ModifyBoneTransforms(
    BoneToModify=((Bone="dyn_baseball",Transform=(Rotation=(X=0,Y=0,Z=0,W=1),Translation=(X=0,Y=0,Z=0),Scale3D=(X=1,Y=1,Z=1))))
    Mode=AdditiveLocal
    Weight=1
    WeightMaximum=1
    WeightMinimum=0
    BoneToModify.0.Transform.Rotation=/Script/RigVM.RigVMFunction_AccumulateQuatLerp(
     Blend=0.35
     bIntegrateDeltaTime=false
     TargetValue=&RigVMFunction_MathQuaternionFromEuler(Euler=(X=0,Y=0,Z=0), Euler.Y=@n1.Result).Result
     InitialValue=&RigVMFunction_MathQuaternionFromEuler(Euler=(X=0,Y=0,Z=0), Euler.Y=@n1.Result).Result
    ) {is=struct RigVMUnitNode, template="AccumulateLerp::Execute(in TargetValue,in InitialValue,in Blend,in bIntegrateDeltaTime,out Result)"}.Result
    ExecutePin=@RigUnit_BeginExecution.ExecutePin
   ) {is=struct RigVMUnitNode}
   &RigUnit_GetBoneTransform(Bone=dyn_baseball, Space=LocalSpace)
  ]}
]
hierarchy [
 b_AR_Root {loc=<0 0 0>, rot=<0 0 0.7071 0.7071>}
 Melee_TwoHanded_Handle {parent=b_AR_Root, loc=<0 0 0>, rot=<0 0 1 0>}
 dyn_rope_knot {parent=Melee_TwoHanded_Handle, loc=<0.173 -0.1224 -16.481>, rot=<0 -0.7071 0 -0.7071>}
 dyn_rope_r_01 {parent=dyn_rope_knot, loc=<1.2402 0.2437 1.2411>, rot=<0 0.0062 0 -1>}
 dyn_rope_r_02 {parent=dyn_rope_r_01, loc=<3.0742 -0.1832 0.2789>, rot=<0 -0.0062 0 -1>}
 dyn_rope_r_03 {parent=dyn_rope_r_02, loc=<5.1165 -0.0009 -0.0553>, rot=<0 0 0 -1>}
 dyn_rope_r_04 {parent=dyn_rope_r_03, loc=<5.238 0 0>, rot=<0 0 0 -1>}
 dyn_rope_l_01 {parent=dyn_rope_knot, loc=<1.338 1.118 0.2073>, rot=<0.9366 0.0116 -0.0313 -0.3488>}
 dyn_rope_l_01_b {parent=dyn_rope_l_01, loc=<2.0282 -0.0239 0.1608>, rot=<0.0338 -0.015 -0.0018 -0.9993>}
 dyn_rope_l_02 {parent=dyn_rope_l_01_b, loc=<2.0922 0 0>, rot=<0 0 0 1>}
 dyn_rope_l_03 {parent=dyn_rope_l_02, loc=<5.1586 0 -0.0393>, rot=<0 -0.0183 0 -0.9998>}
 dyn_teeth_MSTR {parent=Melee_TwoHanded_Handle, loc=<-0.0601 -4.7487 32.1519>, rot=<0 0 0.7071 -0.7071>}
 dyn_teeth_07 {parent=dyn_teeth_MSTR, loc=<8.7712 0.0431 -3.5108>, rot=<0 -0.1076 -0.0001 -0.9942>}
 dyn_teeth_06 {parent=dyn_teeth_MSTR, loc=<6.4649 0.0431 5.2214>, rot=<0 0.0943 -0.0001 -0.9955>}
 dyn_teeth_05 {parent=dyn_teeth_MSTR, loc=<0.1134 0.0431 11.6817>, rot=<0 0.342 -0.0001 -0.9397>}
 dyn_teeth_04 {parent=dyn_teeth_MSTR, loc=<-8.2764 0.043 14.4686>, rot=<-0.0001 0.4884 -0.0002 -0.8726>}
 dyn_teeth_03 {parent=dyn_teeth_MSTR, loc=<-15.8193 0.0431 15.129>, rot=<-0.0002 0.5432 -0.0002 -0.8396>}
 dyn_teeth_02 {parent=dyn_teeth_MSTR, loc=<-22.0383 0.043 15.0875>, rot=<-0.0003 0.5663 -0.0004 -0.8242>}
 dyn_teeth_01 {parent=dyn_teeth_MSTR, loc=<-26.6975 0.043 14.8003>, rot=<-0.0004 0.5733 -0.0005 -0.8193>}
 dyn_baseball_MSTR {parent=Melee_TwoHanded_Handle, loc=<-0.0601 13.8921 24.8762>, rot=<0 0 0.7071 -0.7071>}
 dyn_baseball_MSTR_01 {parent=dyn_baseball_MSTR, loc=<6.4668 0 3.4184>, rot=<0 0 0 1>}
 dyn_baseball {parent=dyn_baseball_MSTR_01, loc=<7.1656 0 0>, rot=<0 0 0 1>}
]
